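; SunnyLoma physics ball, using Rocket League materials, that respawns when moved too far from spawn and celebrates high knocks.
; An authority timer checks 2D squared distance, then triggers a delayed teleport, optionally ground-checked, that zeroes velocity and restarts checking.
; Replicated visibility drives a masked-material wipe through its OnRep.
; Overlaps by tag-eligible actors start a delayed height check, which fires a celebration GameplayCue if the ball rose enough.
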(event EventBeginPlay
  (|CallParentFunction)
  (Variables|Default|SetRespawnTriggerDistanceSquared (Math|Float|Square (ScalableFloat|GetValueatLevel (Variables|Default|GetRespawn%20Trigger%20Distance))))
  ; Store starting location for later respawn and hide the ball prior to wipe.
  (Variables|Default|SetSpawn%20Location (Transformation|GetActorLocation self))
  (CallFunction|InitBall))

(event EventEndPlay
  (params (EndPlayReason :type EEndPlayReason%20Enum :default "Destroyed"))
  (|CallParentFunction EndPlayReason)
  (Networking|SwitchHasAuthority
    (:Authority
      (Utilities|Time|ClearandInvalidateTimerbyHandle (Variables|Default|GetTimerHandleCheckDistance))
      (Utilities|Time|ClearandInvalidateTimerbyHandle (Variables|Default|GetTimerHandleRespawnBall)))))

(event Custom|InitBall
  (if (> (ScalableFloat|GetValueatLevel (Variables|Default|GetB%20Respawn%20on%20Reaching%20Distance)) 0)
    ; Check at intervals whether the ball is too far from its spawn point and if so, respawn it.
    (Networking|SwitchHasAuthority
      (:Authority
        (bind _returnvalue (Utilities|Time|SetTimerbyEvent (delegate Custom|CheckDistanceFromSpawnPoint) (ScalableFloat|GetValueatLevel (Variables|Default|GetDistanceCheckFrequency)) true))
        (Variables|Default|SetTimerHandleCheckDistance _returnvalue)))))

(event Custom|CheckDistanceFromSpawnPoint
  (bind _self self)
  (bind _returnvalue (Transformation|GetActorLocation _self))
  ; Check at intervals whether the ball is too far from its spawn point and if so, respawn it.
  (if (Math|Vector|NotEqualExactly(Vector) (Variables|Default|GetPreviousCheckedLocation) _returnvalue)
    (Variables|Default|SetPreviousCheckedLocation _returnvalue)
    (if (> (Math|Vector|Distance2DSquared(Vector) (Variables|Default|GetSpawn%20Location) (Variables|Default|GetPreviousCheckedLocation)) (Variables|Default|GetRespawnTriggerDistanceSquared))
      (CallFunction|RespawnBall)
      (Utilities|Time|ClearandInvalidateTimerbyHandle (Variables|Default|GetTimerHandleCheckDistance)))))

(event Collision|EventActorBeginOverlap
  (params (OtherActor :type Actor%20Object%20Reference))
  (bind _returnvalue_location (Math|Transform|BreakTransform (Transformation|GetActorTransform self)))
  (bind (_returnvalue_location_x _returnvalue_location_y _returnvalue_location_z) (Math|Vector|BreakVector _returnvalue_location))
  ; Check for knocks to start checking whether to play celebration VFX
  (Variables|Default|SetOverlappingActor OtherActor)
  (if (ScalableFloat|GetValueasBool (Variables|Default|GetbSpawnCelebration))
    (Networking|SwitchHasAuthority
      (:Authority
        ; Only celebrate knocks by pawns & vehicles
        (if (> (ScalableFloat|GetValueatLevel (Variables|Default|GetbSpawnCelebration)) 0)
          (Utilities|Casting|CastToFortPawn (Variables|Default|GetOverlappingActor)
            (:then
              _)
            (:CastFailed
              (Utilities|Casting|CastToFortAthenaVehicle (Variables|Default|GetOverlappingActor))))
          ; Check ball relative height after a duration to check if it was bounced high enough to reward with celebration VFX
          (if (GameplayTags|HasAnyMatchingGameplayTags (Utilities|Casting|ToGameplayTagAssetInterface (Variables|Default|GetOverlappingActor)) (Variables|Default|GetTagsEligibleForCelebration))
            (Utilities|Time|SetTimerbyEvent (delegate Custom|Ball%20Bounced) (ScalableFloat|GetValueatLevel (Variables|Default|GetTimeUntilCheckCelebrationHeight)))
            (Variables|Default|SetHeightAtCollision _returnvalue_location_z)))))))

(event Custom|Ball%20Bounced
  (bind _self self)
  (bind _returnvalue_location (Math|Transform|BreakTransform (Transformation|GetActorTransform _self)))
  (bind (_returnvalue_location_x _returnvalue_location_y _returnvalue_location_z) (Math|Vector|BreakVector _returnvalue_location))
  ; Spawn celebration VFX if the ball was knocked high enough
  (if (>= (- _returnvalue_location_z (Variables|Default|GetHeightAtCollision)) (ScalableFloat|GetValueatLevel (Variables|Default|GetMinimumHeightForCelebration)))
    (GameplayCue|ExecuteGameplayCueOnActor(Burst) self (Variables|Default|GetGCCelebrationTag) (Ability|GameplayCue|MakeGameplayCueParameters :NormalizedMagnitude 0.0 :RawMagnitude 0.0 :Location (Math|Vector|MakeVector _returnvalue_location_x _returnvalue_location_y (+ _returnvalue_location_z (Variables|Default|GetCelebrationOffset))) :Normal "0,0,0" :GameplayEffectLevel 1 :AbilityLevel 1 :bReplicateLocationWhenUsingMinimalRepProxy false))))

(event Custom|FadeBall
  (bind _ball (Variables|Default|GetBall))
  (bind _ball_2 (Variables|Default|GetBall))
  (Networking|CanExecuteCosmeticEvents
    (:True
      (Utilities|FlowControl|Sequence
        (:then_0
          ; Mask the ball which is being faded in/out
          (Rendering|Material|SetMaterial _ball_2 0 "/SunnyLomaBallGameplay/Assets/Materials/MI_Ball_RocketLeague_Masked.MI_Ball_RocketLeague_Masked")
          (Rendering|Material|SetMaterial _ball_2 1 "/SunnyLomaBallGameplay/Assets/Materials/MI_Ball_RocketLeagueTiling01_Masked.MI_Ball_RocketLeagueTiling01_Masked")
          (Rendering|Material|SetMaterial _ball_2 2 "/SunnyLomaBallGameplay/Assets/Materials/MI_Ball_RocketLeagueTiling02_Masked.MI_Ball_RocketLeagueTiling02_Masked"))
        (:then_1
          ; Fade the ball
          (if (Variables|Default|GetbBallVisible)
            (bind (_direction _wipevalue) (Timeline|TimelineFade ReverseFromEnd)
              (:Update
                (Rendering|Material|SetScalarParameterValueonMaterials (Variables|Default|GetBall) "Ball_Visibility" _wipevalue))
              (:Finished
                (if (Variables|Default|GetbBallVisible)
                  ; Unmask the ball
                  (Rendering|Material|SetMaterial _ball 0 "/SunnyLomaBallGameplay/Assets/Materials/MI_Ball_RocketLeague.MI_Ball_RocketLeague")
                  (Rendering|Material|SetMaterial _ball 1 "/SunnyLomaBallGameplay/Assets/Materials/MI_Ball_RocketLeagueTiling01.MI_Ball_RocketLeagueTiling01")
                  (Rendering|Material|SetMaterial _ball 2 "/SunnyLomaBallGameplay/Assets/Materials/MI_Ball_RocketLeagueTiling02.MI_Ball_RocketLeagueTiling02"))))
            (else
              (Timeline|TimelineFade PlayFromStart))))))))

(event Custom|RespawnBall
  ; Return the ball to its original location
  (Variables|Default|SetbBallVisible false)
  (bind _returnvalue (Utilities|Time|SetTimerbyEvent (delegate Custom|TeleportBall) (ScalableFloat|GetValueatLevel (Variables|Default|GetRespawnDelay))))
  (Variables|Default|SetTimerHandleRespawnBall _returnvalue))

(event Custom|TeleportBall
  ; Return the ball to its original location
  (if (ScalableFloat|GetValueasBool (Variables|Default|GetbCheckForBuildsOnRespawn))
    (bind _returnvalue (Helpers|FindGroundLocationUsingWorldContextAt self self (Variables|Default|GetSpawn%20Location) 1e+04 -5e+03 "SunnyLomaBall"))
    (Variables|Default|SetValidGroundLocation (+ _returnvalue (Math|Vector|MakeVector 0.0 0.0 (Components|Sphere|GetScaledSphereRadius (Variables|FortPhysicsBall|GetSphereCollision)))))
    (else
      (Variables|Default|SetValidGroundLocation (Variables|Default|GetSpawn%20Location))))
  (Transformation|SetActorLocation (Variables|Default|GetValidGroundLocation) false true)
  (Variables|/Script/Engine.MovementComponent|SetVelocity "0, 0, 0" (Variables|Default|GetProjectileMovement))
  (CallFunction|InitBall)
  (Variables|Default|SetbBallVisible true))

(fn OnRep_bBallVisible
  (CallFunction|FadeBall))

(fn ConstructionScript
  (|CallParentFunction))
